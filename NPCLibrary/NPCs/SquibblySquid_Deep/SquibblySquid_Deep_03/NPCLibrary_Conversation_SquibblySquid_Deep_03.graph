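# Conversation database for the SquibblySquid Deep 03 NPC, providing its fishing-pun greeting and access to shared NPC conversation options.
# A single entry tag leads to a speech node with a default line and contextual lines gated by first-ever conversation, per-match conversation count or worn outfit.
# First-meeting and outfit reactions take their priority from named curve rows.
# The speech node then branches through link nodes to the shared Quests, BasicServices and RandomServices entry points.
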ConversationDatabase /NPCLibrary/NPCs/SquibblySquid_Deep/SquibblySquid_Deep_03/NPCLibrary_Conversation_SquibblySquid_Deep_03
entry_points [{tag=Conversation.Entry.SquibblySquidDeep.03, destinations [#1]}]
nodes [
 ConversationEntryPointNode_7 {
  id=#1
  class=ConversationEntryPointNode
  outputs [#2]
  properties {entryTag {tagName=Conversation.Entry.SquibblySquidDeep.03}}
 }
 Speech_C_4 {
  id=#2
  class=Speech_C
  outputs [#3, #4, #5]
  properties {
   nodeName=""
   generalConfig {
    defaultMessage=Can I bait ye into spendin' some Gold?
    contextualMessages [
     {
      message=Nothin's bitin'... But an old Squibbly's patient.
      contextRequirements [{requirement {class=FortControllerRequirement_IsFirstEverConversationWithNPC}}]
      priority {curve {rowName=Tandem.Speech.Priority.FirstEverInteraction}}
     }
     {
      message=There's always a catch in the Ocean. Just wait.
      contextRequirements [{requirement {
         class=FortControllerRequirement_CheckNPCMatchConversationCount
         properties {nPCUniqueIdTag {tagName=AISpawnerData.Type.Tandem.SquibblySquidDeep.03}}
        }}]
      priority {curve {curveTable=null, rowName=None}}
     }
     {
      message=Can I bait ye into spendin' some Gold?
      contextRequirements [{requirement {
         class=FortControllerRequirement_CheckNPCMatchConversationCount
         properties {nPCUniqueIdTag {tagName=AISpawnerData.Type.Tandem.SquibblySquidDeep.03}}
        }}]
      priority {curve {curveTable=null, rowName=None}}
     }
     {
      message="Ye look mighty a-lure-ing today!"
      contextRequirements [{requirement {
         class=FortControllerRequirement_CheckNPCMatchConversationCount
         properties {nPCUniqueIdTag {tagName=AISpawnerData.Type.Tandem.SquibblySquidDeep.03}}
        }}]
      priority {curve {curveTable=null, rowName=None}}
     }
     {
      message="Warm plates and full hearts!"
      contextRequirements [{requirement {class=FortControllerRequirement_HasCID, properties {allowedCIds [/BRCosmetics/Athena/Items/Cosmetics/Characters/Character_SquibblySquid]}}}]
      priority {curve {rowName=Tandem.Speech.Priority.OutfitReaction}}
     }
     {
      message="Are ye fixin' to swim with the fishes?!"
      contextRequirements [{requirement {
         class=FortControllerRequirement_HasCID
         properties {allowedCIds [&CID_315_Athena_Commando_M_TeriyakiFish]}
        }}]
      priority {curve {rowName=Tandem.Speech.Priority.OutfitReaction}}
     }
    ]
   }
  }
 }
 ConversationLinkNode_2 {id=#5, class=ConversationLinkNode, properties {remoteEntryTag {tagName=Conversation.Entry.RandomServices}}}
 ConversationLinkNode_0 {id=#4, class=ConversationLinkNode, properties {remoteEntryTag {tagName=Conversation.Entry.BasicServices}}}
 ConversationLinkNode_1 {id=#3, class=ConversationLinkNode, properties {remoteEntryTag {tagName=Conversation.Entry.NPCLibrary.Quests}}}
]
